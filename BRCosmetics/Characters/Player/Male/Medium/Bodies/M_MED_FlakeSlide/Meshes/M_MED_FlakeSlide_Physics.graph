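# Physics asset for the FlakeSlide medium male player body, set up for a two-bone dynamic front chain hanging from the upper spine.
# Each chain bone carries a central capsule with two slimmer side capsules, while large capsules on spine_03 to spine_05 form the torso volume.
# Chain joints link dyn_chain_fr_02 through 01 to spine_05 with free second swing and zero-degree twist limits; spine joints rotate freely on all axes.
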PhysicsAsset /BRCosmetics/Characters/Player/Male/Medium/Bodies/M_MED_FlakeSlide/Meshes/M_MED_FlakeSlide_Physics
bodies {
 dyn_chain_fr_01 [
  {shapeType=Capsule, length=6, radius=5.5, shapeName=dyn_chain_fr_01_capsule, center=<2 0 9>, rotation=<rot 0 90 80.0001>}
  {shapeType=Capsule, length=6.8816, radius=3.1771, shapeName=dyn_chain_fr_01_capsule, center=<4.1368 11.6153 1.9923>, rotation=<rot -90 -10 0>}
  {shapeType=Capsule, length=6.6561, radius=2.5462, shapeName=dyn_chain_fr_01_capsule, center=<3.8569 -10.9468 1.2907>, rotation=<rot -90 15 0>}
 ]
 dyn_chain_fr_02 [
  {shapeType=Capsule, length=5, radius=5, shapeName=dyn_chain_fr_02_capsule, center=<4 0 5.4>, rotation=<rot 0 90 75.0001>}
  {
   shapeType=Capsule
   length=8.4326
   radius=2.6068
   shapeName=dyn_chain_fr_02_capsule
   center=<4.2598 6.2797 2.0708>
   rotation=<rot -79.4296 42.0166 -86.0909>
  }
  {
   shapeType=Capsule
   length=8.2621
   radius=2.6955
   shapeName=dyn_chain_fr_02_capsule
   center=<3.9564 -5.6165 1.7404>
   rotation=<rot -82.3794 -67.8062 112.1893>
  }
 ]
 spine_03 [{shapeType=Capsule, length=25, radius=18, shapeName=spine_03_capsule}]
 spine_04 [{shapeType=Capsule, length=25, radius=17.5, shapeName=spine_04_capsule}]
 spine_05 [{shapeType=Capsule, length=25, radius=16, center=<-4.439 -2.6522 0>, rotation=<rot 0 90 0>}]
}
constraints [
 {bone1Name=dyn_chain_fr_01, bone2Name=spine_05, swing1LimitDegrees=25, swing2LimitDegrees=50, swing2Motion=Free, twistLimitDegrees=0}
 {bone1Name=dyn_chain_fr_02, bone2Name=dyn_chain_fr_01, swing1LimitDegrees=25, swing2LimitDegrees=50, swing2Motion=Free, twistLimitDegrees=0}
 {
  bone1Name=spine_05
  bone2Name=spine_04
  swing1LimitDegrees=45
  swing1Motion=Free
  swing2LimitDegrees=45
  swing2Motion=Free
  twistLimitDegrees=45
  twistMotion=Free
 }
 {
  bone1Name=spine_04
  bone2Name=spine_03
  swing1LimitDegrees=45
  swing1Motion=Free
  swing2LimitDegrees=45
  swing2Motion=Free
  twistLimitDegrees=45
  twistMotion=Free
 }
]
