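# Post-process style Animation Blueprint for the FlakeSlide medium male body cosmetic, copying the attached parent mesh's pose and adding secondary motion.
# Two Control Rigs (base, crouching) feed spine-relative RigidBody physics, a zipper constraint and AnimDynamics for sleeves and zipper.
# Additive bone offsets then adjust collar, chain and state-dependent backpack attachment.
# The EventGraph derives crouch, emote, front-end, targeting, pistol and head-angle values, several driving those nodes.
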AnimBlueprint /BRCosmetics/Characters/Player/Male/Medium/Bodies/M_MED_FlakeSlide/Meshes/M_MED_FlakeSlide_AnimBP
target_skeleton=&/M_MED_FlakeSlide_Skeleton
graphs [
 AnimGraph {nodes [
   n1 = Control Rig!(
    Source=Copy Pose From Mesh(SourceMeshComponent=Get Skeletal_Mesh() -> (Skeletal_Mesh:Skeletal Mesh Component Object Reference)) {bUseAttachedParent=true}
    Alpha=Get CrouchAlpha() -> (CrouchAlpha:double)
    Playing_Emots:bool=Get bIsPlayingEmote()
    ParachuteOpened:bool=Get bIsParachuteOpened()
    isPlayingEmoteFloat:double=Get is Playing Emote() -> (is Playing Emote:double)
   ) -> (Pose) {controlRigAssetReference {blueprintRigClass=&/M_MED_FlakeSlide_CtrlRig.M_MED_FlakeSlide_CtrlRig_C}}
   n2 = Control Rig!(
    Source=@n1
    Alpha
    Crouching:bool=Get bIsCrouching()
    CrouchAlpha:double=Get IsCrouching() -> (IsCrouching:double)
    CrouchWalkHeadup_Bool:bool=Get CrouchWalkHeadup_Bool() -> (CrouchWalkHeadup_Bool:bool)
    CrouchHeadup_Bool:bool=Get CrouchHeadup_Bool() -> (CrouchHeadup_Bool:bool)
    Not_in_head_up:bool=Get HeadUp() -> (HeadUp:bool)
    targeting:double=Get isTargeting() -> (isTargeting:double)
    Pistol:double=Get isUsingPistol() -> (isUsingPistol:double)
   ) -> (Pose) {controlRigAssetReference {blueprintRigClass=&/M_MED_FlakeSlide_CtrlRig_Crouching.M_MED_FlakeSlide_CtrlRig_Crouching_C}}
   n3 = RigidBody!(
    ComponentPose=Local To Component(LocalPose=@n2)
    Alpha
    OverrideWorldGravity:Vector="0,0,0"
    ExternalForce
    SimSpaceSettings:Sim Space Settings Structure=(WorldAlpha=0,VelocityScaleZ=1,DampingAlpha=1,MaxLinearVelocity=10000,MaxAngularVelocity=10000,MaxLinearAcceleration=10000,MaxAngularAcceleration=10000,ExternalLinearDragV=(X=0,Y=0,Z=0),ExternalLinearVelocity=(X=0,Y=0,Z=0),ExternalAngularVelocity=(X=0,Y=0,Z=0))
   ) -> (Pose) {
    bOverrideWorldGravity=true
    overridePhysicsAsset=&/M_MED_FlakeSlide_Physics
    simulationSpace=BaseBoneSpace
    baseBoneRef {boneName=spine_05}
   }
   n4 = "Constraint - dyn_zip_fr_01 (PPPP)"!(
    ComponentPose=@n3
    Alpha=1
    ConstraintWeights_0=0.8052
    ConstraintWeights_1:float=0.0878
    ConstraintWeights_2:float=0.0823
    ConstraintWeights_3:float=0.0247
   ) -> (Pose) {
    boneToModify {boneName=dyn_zip_fr_01}
    constraintSetup [
     {targetBone {boneName=dyn_coller_fr_01}}
     {targetBone {boneName=clavicle_l}}
     {targetBone {boneName=clavicle_r}}
     {targetBone {boneName=spine_04}}
    ]
    constraintWeights [1,1,1,1]
   }
   n5 = Anim Dynamics - Bone: base_calf_sleeve_01_r!(ComponentPose=@n4, Alpha=1, GravityOverride:Vector="0,0,-980", ExternalForce="0,0,0") -> (Pose) {
    angularDampingOverride=1
    angularSpringConstant=8000
    bAngularSpring=true
    bOverrideAngularDamping=true
    bUseGravityOverride=true
    gravityScale=0.35
    simulationSpace=RootRelative
    boundBone {boneName=base_calf_sleeve_01_r}
    componentAppliedLinearAccClamp=<5000 5000 5000>
    componentLinearAccScale=<5 5 5>
    componentLinearVelScale=<5 5 5>
    gravityOverride {z=-980}
    physicsBodyDefinitions [{
      boundBone {boneName=base_calf_sleeve_01_r}
      boxExtents=<20 30 30>
      constraintSetup {angularLimitsMax=<0 50 50>, angularLimitsMin=<0 -50 -50>, angularTarget=<1 0 0>}
      localJointOffset=<4 0 0>
     }]
   }
   n6 = Anim Dynamics - Bone: base_calf_sleeve_01_l!(ComponentPose=@n5, Alpha=1, GravityOverride:Vector="0,0,-980", ExternalForce="0,0,0") -> (Pose) {
    angularDampingOverride=1
    angularSpringConstant=8000
    bAngularSpring=true
    bOverrideAngularDamping=true
    bUseGravityOverride=true
    gravityScale=0.35
    simulationSpace=RootRelative
    boundBone {boneName=base_calf_sleeve_01_l}
    componentAppliedLinearAccClamp=<5000 5000 5000>
    componentLinearAccScale=<5 5 5>
    componentLinearVelScale=<5 5 5>
    gravityOverride {z=-980}
    physicsBodyDefinitions [{
      boundBone {boneName=base_calf_sleeve_01_l}
      boxExtents=<20 30 30>
      constraintSetup {angularLimitsMax=<0 50 50>, angularLimitsMin=<0 -50 -50>, angularTarget=<1 0 0>}
      localJointOffset=<4 0 0>
     }]
   }
   n7 = Anim Dynamics - Bone: base_lowerarm_sleeve_01_l!(ComponentPose=@n6, Alpha=1, GravityOverride:Vector="0,0,-980", ExternalForce="0,0,0") -> (Pose) {
    angularDampingOverride=1
    angularSpringConstant=2000
    bAngularSpring=true
    bOverrideAngularDamping=true
    bUseGravityOverride=true
    gravityScale=0.35
    simulationSpace=RootRelative
    alphaBoolBlend {blendOption=Linear}
    boundBone {boneName=base_lowerarm_sleeve_01_l}
    componentAppliedLinearAccClamp=<5000 5000 5000>
    componentLinearAccScale=<0.5 0.5 0.5>
    componentLinearVelScale=<0.5 0.5 0.5>
    gravityOverride {z=-980}
    physicsBodyDefinitions [{
      boundBone {boneName=base_lowerarm_sleeve_01_l}
      boxExtents=<15 15 15>
      constraintSetup {angularLimitsMax=<0 8 8>, angularLimitsMin=<0 -8 -8>, angularTarget=<1 0 0>}
      localJointOffset=<1 0 0>
     }]
   }
   n8 = Anim Dynamics - Bone: base_lowerarm_sleeve_01_r!(ComponentPose=@n7, Alpha=1, GravityOverride:Vector="0,0,-980", ExternalForce="0,0,0") -> (Pose) {
    angularDampingOverride=1
    angularSpringConstant=2000
    bAngularSpring=true
    bOverrideAngularDamping=true
    bUseGravityOverride=true
    gravityScale=0.35
    simulationSpace=RootRelative
    alphaBoolBlend {blendOption=Linear}
    boundBone {boneName=base_lowerarm_sleeve_01_r}
    componentAppliedLinearAccClamp=<5000 5000 5000>
    componentLinearAccScale=<0.5 0.5 0.5>
    componentLinearVelScale=<0.5 0.5 0.5>
    gravityOverride {z=-980}
    physicsBodyDefinitions [{
      boundBone {boneName=base_lowerarm_sleeve_01_r}
      boxExtents=<15 15 15>
      constraintSetup {angularLimitsMax=<0 8 8>, angularLimitsMin=<0 -8 -8>, angularTarget=<1 0 0>}
      localJointOffset=<1 0 0>
     }]
   }
   n9 = Anim Dynamics - Bone: dyn_zip_fr_01!(ComponentPose=@n8, Alpha=1, GravityOverride:Vector="0,0,-980", ExternalForce="0,0,0") -> (Pose) {
    angularDampingOverride=1
    angularSpringConstant=4500
    bAngularSpring=true
    bOverrideAngularDamping=true
    bUseGravityOverride=true
    gravityScale=0.35
    simulationSpace=RootRelative
    alphaBoolBlend {blendOption=Linear}
    boundBone {boneName=dyn_zip_fr_01}
    componentAppliedLinearAccClamp=<2000 2000 2000>
    componentLinearAccScale=<0.5 0.5 0.5>
    componentLinearVelScale=<0.5 0.5 0.5>
    gravityOverride {z=-980}
    physicsBodyDefinitions [{
      boundBone {boneName=dyn_zip_fr_01}
      boxExtents=<15 15 15>
      constraintSetup {angularLimitsMax=<5 5 10>, angularLimitsMin=<-5 -60 -10>, angularTarget=<1 0 0>}
      localJointOffset=<4 0 0>
     }]
   }
   n10 = "Transform (Modify) Bone - Bone: dyn_coller_a_01_r"(ComponentPose=@n9, Alpha=1, Translation="0,0,0", Rotation="30,0,5", Scale="1,1,1") {
    rotationMode=BMM_Additive
    rotationSpace=BCS_BoneSpace
    scaleMode=BMM_Additive
    scaleSpace=BCS_BoneSpace
    translationMode=BMM_Additive
    translationSpace=BCS_BoneSpace
    alphaScaleBiasClamp {bInterpResult=true}
    boneToModify {boneName=dyn_coller_a_01_r}
    rotation {pitch=30, roll=5}
   }
   n11 = "Transform (Modify) Bone - Bone: dyn_coller_a_01_l"(ComponentPose=@n10, Alpha=1, Translation="0,0,0", Rotation="30,4.059,18.488", Scale="1,1,1") {
    rotationMode=BMM_Additive
    rotationSpace=BCS_BoneSpace
    scaleMode=BMM_Additive
    scaleSpace=BCS_BoneSpace
    translationMode=BMM_Additive
    translationSpace=BCS_BoneSpace
    alphaScaleBiasClamp {bInterpResult=true}
    boneToModify {boneName=dyn_coller_a_01_l}
    rotation=<rot 30 4.059 18.488>
   }
   n12 = "Transform (Modify) Bone - Bone: dyn_coller_fr_01"(ComponentPose=@n11, Alpha=1, Translation="0,0,0", Rotation="40,0,5.095", Scale="1,1,1") {
    rotationMode=BMM_Additive
    rotationSpace=BCS_BoneSpace
    scaleMode=BMM_Additive
    scaleSpace=BCS_BoneSpace
    translationMode=BMM_Additive
    translationSpace=BCS_BoneSpace
    alphaScaleBiasClamp {bInterpResult=true}
    boneToModify {boneName=dyn_coller_fr_01}
    rotation {pitch=40, roll=5.095}
   }
   n13 = "Transform (Modify) Bone - Bone: dyn_chain_fr_01"!(ComponentPose=@n12, bAlphaBoolEnabled:bool=True, Translation="0,0,0", Rotation="20.000,0.009,-0.010", Scale="1,1,1") -> (Pose) {
    alphaInputType=Bool
    rotationMode=BMM_Additive
    rotationSpace=BCS_BoneSpace
    translationMode=BMM_Additive
    translationSpace=BCS_BoneSpace
    boneToModify {boneName=dyn_chain_fr_01}
    rotation=<rot 20 0.009 -0.01>
   }
   n14 = "Transform (Modify) Bone - Bone: attach_backpack"!(ComponentPose=@n13, bAlphaBoolEnabled:bool=Get bIsSprinting(), Translation="0,-2.5,0", Rotation="0,0,0", Scale="1,1,1") -> (Pose) {
    alphaInputType=Bool
    rotationMode=BMM_Additive
    rotationSpace=BCS_BoneSpace
    scaleMode=BMM_Additive
    scaleSpace=BCS_BoneSpace
    translationMode=BMM_Additive
    translationSpace=BCS_BoneSpace
    alphaBoolBlend {blendInTime=1, blendOutTime=1}
    boneToModify {boneName=attach_backpack}
    translation {y=-2.5}
   }
   n15 = "Transform (Modify) Bone - Bone: attach_backpack"(ComponentPose=@n14, bAlphaBoolEnabled=Get bIsSkyDiving(), Translation="0,-4,0", Rotation="0,0,4", Scale="1,1,1") {
    alphaInputType=Bool
    rotationMode=BMM_Additive
    rotationSpace=BCS_BoneSpace
    scaleMode=BMM_Additive
    scaleSpace=BCS_BoneSpace
    translationMode=BMM_Additive
    translationSpace=BCS_BoneSpace
    alphaBoolBlend {blendInTime=1, blendOutTime=1}
    boneToModify {boneName=attach_backpack}
    rotation {roll=4}
    translation {y=-4}
   }
   n16 = "Transform (Modify) Bone - Bone: attach_backpack"!(ComponentPose=@n15, Alpha:float=Get is Playing Emote(), Translation="0,0.5,0", Rotation="0,0,0", Scale="1,1,1") -> (Pose) {
    rotationMode=BMM_Additive
    rotationSpace=BCS_BoneSpace
    scaleMode=BMM_Additive
    scaleSpace=BCS_BoneSpace
    translationMode=BMM_Additive
    translationSpace=BCS_BoneSpace
    alphaBoolBlend {blendInTime=1, blendOutTime=1}
    boneToModify {boneName=attach_backpack}
    translation {y=0.5}
   }
   n17 = "Transform (Modify) Bone - Bone: attach_backpack"(ComponentPose=@n16, Alpha=Get FrontEndidle() -> (FrontEndidle:double), Translation="0,2.5,0", Rotation="0,0,0", Scale="1,1,1") {
    rotationMode=BMM_Additive
    rotationSpace=BCS_BoneSpace
    scaleMode=BMM_Additive
    scaleSpace=BCS_BoneSpace
    translationMode=BMM_Additive
    translationSpace=BCS_BoneSpace
    alphaBoolBlend {blendInTime=1, blendOutTime=1}
    boneToModify {boneName=attach_backpack}
    translation {y=2.5}
   }
   Output Pose(Result=Component To Local(ComponentPose=@n17)) {layerGroup=None}
  ]}
 EventGraph {nodes [
   Event BlueprintUpdateAnimation()
   > n1 = Sequence!(execute) -> (then_0, then_1, then_2:Exec, then_3:Exec, then_4:Exec, then_5:Exec)
   > n2 = "Gravity Override State Function (Parameter Asset) (Nativized)"(
    CharacterPartAnimInstance=Self-Reference()
    GameTimeInSeconds=GetGameTimeInSeconds()
    Parameters=&/M_MED_FlakeSlide_Parameters
   )
   > Set GravityOverride(GravityOverride=@n2)
   > Set RBAN Settings Out(RBAN Settings Out=@n2.OutRigidBodyAnimNodeParameters)
   Set crouch Walk(execute=@n1.then_2, crouch Walk:bool=AND Boolean(A=Get bIsCrouching(), B=Get bIsMoving2D())) -> (Output_Get:bool)
   n3 = Branch(execute=@n1.then_3, Condition=Get bIsCrouching())
   > Set CrouchAlpha(CrouchAlpha:double=0) -> (Output_Get:double)
   Set CrouchAlpha(execute=@n3.else, CrouchAlpha=1)
   Set CrouchHeadup_Bool(execute=@n1.then_5, CrouchHeadup_Bool:bool=OR Boolean!(A=Get HeadUp(), B=Get bIsTargeting(), C:bool=Get bIsMoving2D()) -> (ReturnValue)) -> (Output_Get:bool)
   n4 = NOT Boolean(A=Get bIsPlayingEmote())
   n5 = Get bIsInFrontEnd()
   Set FrontEndidle(execute=@n1.then_4, FrontEndidle:double="To Float (Boolean)"(InBool=AND Boolean(A=@n5, B=@n4))) -> (Output_Get:double)
   > Set is Playing Emote(is Playing Emote:double="To Float (Boolean)"(InBool=Get bIsPlayingEmote())) -> (Output_Get:double)
   n6 = Get bIsCrouching()
   n7 = Get bIsMoving2D()
   > Set CrouchWalkHeadup_Bool(CrouchWalkHeadup_Bool:bool="Equal (Float)"(A=Get Head Up() -> (Head Up:double), B:bool=AND Boolean(A=@n6, B=@n7))) -> (Output_Get:bool)
   > Set isTargeting(isTargeting:double="To Float (Boolean)"(InBool=Get bIsTargeting())) -> (Output_Get:double)
   n8 = GetOwnerSkeletalMeshComponent()
   Set Head Up(
    execute=@n1.then_1
    Head Up:double=GetAlphaFromReferenceTransformDelta(SkeletalMeshComponent=@n8, BoneName=spine_05, BaseName=head, InputRange=(X=10,Y=70), TwistAxis="0,0,1")
   ) -> (Output_Get:double)
   n9 = Get bIsPlayingEmote()
   > Set Head Down(Head Down:double=float * float(
     A=GetAlphaFromReferenceTransformDelta(SkeletalMeshComponent=@n8, BoneName=spine_05, BaseName="head ", DeltaComponentType=EulerY, InputRange=(X=-4,Y=-9), TwistAxis="0,0,1")
     B:bool=NOT Boolean(A=@n9)
    ) {is=title K2Node_PromotableOperator}) -> (Output_Get:double)
   > Set Head_Turn_L(Head_Turn_L:double=GetAlphaFromReferenceTransformDelta(SkeletalMeshComponent=@n8, BoneName=spine_05, BaseName=head, DeltaComponentType=EulerX, InputRange=(X=0,Y=-22), TwistAxis="1,0,0")) -> (Output_Get:double)
   > Set Head_Turn_R(Head_Turn_R:double=GetAlphaFromReferenceTransformDelta(SkeletalMeshComponent=@n8, BoneName=spine_05, BaseName=head, DeltaComponentType=EulerX, InputRange=(X=0,Y=22), TwistAxis="1,0,0")) -> (Output_Get:double)
   Event OnBeginCrouching()
   > Set IsCrouching(IsCrouching:double=1) -> (Output_Get:double)
   Event OnEndCrouching()
   > Set IsCrouching(IsCrouching=0)
   Event OnWeaponChanged()
   > Set isUsingPistol(isUsingPistol:double=0) -> (Output_Get:double)
   > n10 = Switch on EFortWeaponCoreAnimation!(execute, Selection:EFortWeaponCoreAnimation Enum=Get WeaponCoreAnim!() -> (WeaponCoreAnim)) -> (
    Melee:Exec
    Pistol:Exec
    Shotgun:Exec
    PaperBlueprint:Exec
    Rifle:Exec
    MeleeOneHand:Exec
    MachinePistol:Exec
    RocketLauncher:Exec
    GrenadeLauncher:Exec
    GoingCommando:Exec
    AssaultRifle:Exec
    TacticalShotgun:Exec
    SniperRifle:Exec
    TrapPlacement:Exec
    ShoulderLauncher:Exec
    AbilityDecoTool:Exec
    Crossbow:Exec
    C4:Exec
    RemoteControl:Exec
    DualWield:Exec
    AR_BullPup:Exec
    AR_ForwardGrip:Exec
    MedPackPaddles:Exec
    SMG_P90:Exec
    AR_DrumGun:Exec
    Consumable_Small:Exec
    Consumable_Large:Exec
    Balloon:Exec
    MountedTurret:Exec
    CreativeTool:Exec
    ExplosiveBow:Exec
    AshtonIndigo:Exec
    AshtonChicago:Exec
    MeleeDualWield:Exec
    Unarmed:Exec
   )
   Set isUsingPistol(execute=@n10.Pistol, isUsingPistol=1)
  ]}
]
variables [
 Aim_down {type=boolean}
 AimUp {type=number}
 crouch Walk {type=boolean}
 CrouchAlpha {type=number}
 CrouchHeadup_Bool {type=boolean}
 CrouchTargeting {type=boolean}
 CrouchWalkHeadup {type=number}
 CrouchWalkHeadup_Bool {type=boolean}
 FrontEndidle {type=number}
 GravityOverride {type=Vector}
 Head Aim Down {type=number}
 Head down {type=number}
 Head Down_Bool {type=boolean}
 Head look up alpha {type=number}
 Head Up {type=number}
 Head_Turn_L {type=number}
 Head_Turn_R {type=number}
 HeadUp {type=boolean}
 is Playing Emote {type=number}
 IsCrouching {type=number}
 isTargeting {type=number}
 isUsingPistol {type=number}
 RBAN Settings Out {type=FortRigidBodyAnimNodeParameters}
 Skel Mesh {type=SkeletalMeshComponent}
 Skeletal Mesh {type=SkeletalMesh, default=null}
 Skeletal_Mesh {type=SkeletalMeshComponent}
]
