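# Animation Blueprint for the FlakeSlide male character's face accessory part, adding secondary motion to hair and goggle bones.
# It copies the pose from the attached parent mesh, then chains spring-based Anim Dynamics nodes through three hair bones and one half-weighted, tightly limited goggle bone.
# Each update, a parameter-asset function sets gravity and rigid-body settings variables, but the AnimGraph uses fixed gravity values and has no rigid-body node.
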AnimBlueprint /BRCosmetics/Characters/Player/Male/Medium/Bodies/M_MED_FlakeSlide/Meshes/Parts/M_MED_FlakeSlide_FaceAcc_AnimBP
target_skeleton=&/M_MED_FlakeSlide_FaceAcc_Skeleton
graphs [
 AnimGraph {nodes [
   n1 = Get Skeletal_Mesh() -> (Skeletal_Mesh:Skeletal Mesh Component Object Reference)
   n2 = Anim Dynamics - Bone: dyn_hair_a01!(
    ComponentPose=Local To Component(LocalPose=Copy Pose From Mesh(SourceMeshComponent=@n1) {bUseAttachedParent=true})
    Alpha=1
    GravityOverride:Vector="0,0,-2000"
    ExternalForce="0,0,0"
   ) -> (Pose) {
    angularDampingOverride=0.9999
    angularSpringConstant=4000
    bAngularSpring=true
    bOverrideAngularDamping=true
    bUseGravityOverride=true
    gravityScale=0.35
    simulationSpace=RootRelative
    alphaBoolBlend {blendOption=Linear}
    boundBone {boneName=dyn_hair_a01}
    componentAppliedLinearAccClamp=<3000 3000 3000>
    componentLinearAccScale=<0.5 0.5 0.5>
    componentLinearVelScale=<0.5 0.5 0.5>
    gravityOverride {z=-2000}
    physicsBodyDefinitions [{
      boundBone {boneName=dyn_hair_a01}
      boxExtents=<20 25 25>
      constraintSetup {angularLimitsMax=<0 -5 5>, angularLimitsMin=<0 -10 -5>, angularTarget=<1 0 0>}
      localJointOffset=<3 0 0>
     }]
   }
   n3 = Anim Dynamics - Bone: dyn_hair_a02!(ComponentPose=@n2, Alpha=1, GravityOverride:Vector="0,0,-2000", ExternalForce="0,0,0") -> (Pose) {
    angularDampingOverride=0.9999
    angularSpringConstant=4000
    bAngularSpring=true
    bOverrideAngularDamping=true
    bUseGravityOverride=true
    gravityScale=0.35
    simulationSpace=RootRelative
    alphaBoolBlend {blendOption=Linear}
    boundBone {boneName=dyn_hair_a02}
    componentAppliedLinearAccClamp=<3000 3000 3000>
    componentLinearAccScale=<0.5 0.5 0.5>
    componentLinearVelScale=<0.5 0.5 0.5>
    gravityOverride {z=-2000}
    physicsBodyDefinitions [{
      boundBone {boneName=dyn_hair_a02}
      boxExtents=<20 25 25>
      constraintSetup {angularLimitsMax=<0 5 5>, angularLimitsMin=<0 -10 -5>, angularTarget=<1 0 0>}
      localJointOffset=<3 0 0>
     }]
   }
   n4 = Anim Dynamics - Bone: dyn_hair_b01!(ComponentPose=@n3, Alpha=1, GravityOverride:Vector="0,0,-2000", ExternalForce="0,0,0") -> (Pose) {
    angularDampingOverride=0.9999
    angularSpringConstant=4000
    bAngularSpring=true
    bOverrideAngularDamping=true
    bUseGravityOverride=true
    gravityScale=0.35
    simulationSpace=RootRelative
    alphaBoolBlend {blendOption=Linear}
    boundBone {boneName=dyn_hair_b01}
    componentAppliedLinearAccClamp=<3000 3000 3000>
    componentLinearAccScale=<0.5 0.5 0.5>
    componentLinearVelScale=<0.5 0.5 0.5>
    gravityOverride {z=-2000}
    physicsBodyDefinitions [{
      boundBone {boneName=dyn_hair_b01}
      boxExtents=<10 10 10>
      constraintSetup {angularLimitsMax=<0 15 3>, angularLimitsMin=<0 -15 -3>, angularTarget=<1 0 0>}
      localJointOffset=<3 0 0>
     }]
   }
   n5 = Anim Dynamics - Bone: dyn_google_a01!(ComponentPose=@n4, Alpha=0.5, GravityOverride:Vector="0,0,0", ExternalForce="0,0,0") -> (Pose) {
    angularSpringConstant=20000
    bAngularSpring=true
    bUseGravityOverride=true
    boundBone {boneName=dyn_google_a01}
    componentAppliedLinearAccClamp=<3000 3000 3000>
    componentLinearAccScale=<0.5 0.5 0.5>
    componentLinearVelScale=<0.5 0.5 0.5>
    physicsBodyDefinitions [{
      boundBone {boneName=dyn_google_a01}
      boxExtents=<10 10 10>
      constraintSetup {angularLimitsMax=<0 0.6 0>, angularLimitsMin=<0 -0.5 0>, angularTarget=<0 0 0>}
      localJointOffset=<0 0 0>
     }]
   }
   Output Pose(Result=Component To Local(ComponentPose=@n5)) {layerGroup=None}
  ]}
 EventGraph {nodes [
   Event BlueprintUpdateAnimation()
   > n1 = "Gravity Override State Function (Parameter Asset) (Nativized)"(
    CharacterPartAnimInstance=Self-Reference()
    GameTimeInSeconds=GetGameTimeInSeconds()
    Parameters=&/M_MED_FlakeSlide_FaceAcc_Parameters
   )
   > Set GravityOverride(GravityOverride=@n1)
   > Set RBAN Settings Out(RBAN Settings Out=@n1.OutRigidBodyAnimNodeParameters)
   Event BlueprintInitializeAnimation()
   > Set Skeletal_Mesh(Skeletal_Mesh:Skeletal Mesh Component Object Reference=GetOwnerSkeletalMeshComponent()) -> (Output_Get:Skeletal Mesh Component Object Reference)
  ]}
]
variables [
 GravityOverride {type=Vector}
 RBAN Settings Out {type=FortRigidBodyAnimNodeParameters}
 Skeletal_Mesh {type=SkeletalMeshComponent}
]
